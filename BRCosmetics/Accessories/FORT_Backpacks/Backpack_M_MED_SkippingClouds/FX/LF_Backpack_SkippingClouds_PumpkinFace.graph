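# Light function material for the Skipping Clouds backpack that masks a light with a texture, named as a pumpkin face projection.
# It flips the U coordinate, scales the UVs around their center by the Scale parameter, and samples a swappable texture that defaults to a stacked-plane glow.
# A static switch chooses between the raw RGB and a SmoothStep remap using Min and Max parameters.
# The material is forced compatible with the light function atlas.
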Material /BRCosmetics/Accessories/FORT_Backpacks/Backpack_M_MED_SkippingClouds/FX/LF_Backpack_SkippingClouds_PumpkinFace
graph {
 properties {
  bForceCompatibleWithLightFunctionAtlas=true
  materialDomain=MD_LightFunction
  pixelDepthOffsetMode=PDOM_Legacy
  usedShadingModels=Default Lit
 }
 nodes [
  n1 = TextureCoordinate()
  n2 = OneMinus(None=ComponentMask(None=@n1) {r=true})
  n3 = ComponentMask(None=@n1) {g=true}
  n4 = TextureSampleParameter2D(UVs=MaterialFunctionCall(UVs=AppendVector(A=@n2, B=@n3), Texture Scale=ScalarParameter() {defaultValue=1, parameterName=Scale, sliderMax=10}) {materialFunction=&ScaleUVsByCenter}.UVs) {
   parameterName=TextureInput
   texture=&/T_Backpack_SkippingClouds_StackedPlaneGlow
  }
  StaticSwitchParameter(
   True=SmoothStep(Min=ScalarParameter() {parameterName=Min, sliderMax=1}, Max=ScalarParameter() {parameterName=Max, sliderMax=1}, Value=@n4.RGB)
   False=@n4.RGB
  ) {parameterName=UseSmoothStep?}
 ]
}
