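# Additive, unlit, two-sided Niagara sprite and mesh-particle material for the stacked glow effect on the SkippingClouds backpack cosmetic.
# UVs are centered, scaled and bent by a spherical Curvature projection before sampling an SDF texture.
# A smoothstep on the selected channel, driven by the particle's SDF Fade dynamic parameter, shapes the glow.
# Particle color feeds emissive adjustments; particle alpha multiplies the shape, a switchable view-angle fade and optional depth fade.
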Material /BRCosmetics/Accessories/FORT_Backpacks/Backpack_M_MED_SkippingClouds/FX/M_Backpack_SkippingClouds_StackedGlow
graph {
 properties {
  bUsedWithNiagaraMeshParticles=true
  bUsedWithNiagaraSprites=true
  blendMode=BLEND_Additive
  pixelDepthOffsetMode=PDOM_Legacy
  shadingModel=MSM_Unlit
  twoSided=true
  usedShadingModels=Unlit
 }
 nodes [
  n1 = ParticleColor()
  MaterialFunctionCall(Emissive=@n1.RGB) {materialFunction=&MF_FX_EmissiveAdjustments}
  n2 = DynamicParameter() {defaultValue=<rgba 0 1 1 1>, paramNames [SDF Fade, Param2, Param3, Param4]}
  n3 = Subtract(A=ScalarParameter() {parameterName=Min}, B=@n2.SDF Fade)
  n4 = ScalarParameter() {parameterName=SDF Fade, sliderMax=2}
  n5 = Divide(A=ConstantBiasScale(None=TextureCoordinate()) {bias=-0.5, scale=2}, B=ScalarParameter() {defaultValue=1, parameterName=Size, sliderMax=2}) {constB=1}
  n6 = Multiply(A=@n5, B=ScalarParameter() {defaultValue=0.5, parameterName=Curvature, sliderMax=2})
  n7 = OneMinus(None=Saturate(None=DotProduct(A=@n6, B=@n6)))
  n8 = Divide(A=@n5, B=Add(A=SquareRoot(None=@n7)))
  n9 = TextureSampleParameter2D(UVs=Saturate(None=ConstantBiasScale(None=@n8))) {
   parameterName=Texture
   texture=&/T_Backpack_SkippingClouds_StackedPlaneGlow_SDF
  }
  n10 = SmoothStep(
   Min=Saturate(None=@n3)
   Max=Add(A=Add(A=@n3, B=@n4), B=@n2.SDF Fade)
   Value=ChannelMaskParameter(None=@n9.RGBA) {parameterName=Channel Mask}.R
  )
  n11 = SmoothStep(
   Min=ScalarParameter() {parameterName=ViewFadeStart, sliderMax=1}
   Max=ScalarParameter() {defaultValue=1, parameterName=ViewFadeEnd, sliderMax=1}
  )
  n12 = MaterialFunctionCall("FadeSensitivity (optional)"=ScalarParameter() {defaultValue=2, parameterName=ViewAngleFadeSensitivity, sliderMax=10}) {materialFunction=&FadeBasedOnViewAngle}
  n13 = ScalarParameter() {parameterName=FadeDistance, sliderMax=30}
  Multiply(
   A=Multiply(A=Multiply(A=@n1.A, B=@n10), B=StaticSwitchParameter(True=@n11, False=@n12.Result) {parameterName=UseViewFadeSmoothStep})
   B=StaticSwitchParameter(True=DepthFade(FadeDistance=@n13), False=Constant() {r=1}) {parameterName=UseDepthFade}
  )
 ]
}
